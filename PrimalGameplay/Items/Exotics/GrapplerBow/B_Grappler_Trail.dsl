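; Visual rope actor for the Grappler Bow, stretched between the weapon's muzzle and the secondary projectile's attach socket.
; Tick moves, rotates and scales the rope mesh, setting material length until the projectile is gone.
; A timeline then reels the arrow back, sends a return gameplay event and destroys the actor; another drives rope wiggle for tightening.
; Destruction or pre-invalidation of the pawn's ability system removes rope-active and pickup-blocking tags.
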(event EventBeginPlay
  (bind _playerpawn (Variables|Default|GetPlayerPawn))
  ;[0.5448 1 0.8118] EVENT -> BeginPlay
  (Actor|Tick|SetActorTickEnabled true)
  (Utilities|IsValid _playerpawn
    (:Is%20Valid
      (EventDispatchers|BindEventtoOnPreInvalidateAbilitySystemComponent _playerpawn (EventDispatchers|CreateEvent :!event Custom|OnPlayerPawnPreInvalidateAbilitySystemComponent)))))

(event EventDestroyed
  ;[1 0.6595 0.4235] Event Destroyed and OnPrePlayerPawnDetachAbilitySystemComponent - Remove Rope Active and Block Incoming Pickups tags
  (Utilities|IsValid (Variables|Default|GetPlayerPawn)
    (:Is%20Valid
      (Tags|RemoveNonReplicatedTagtoPawn (Variables|Default|GetPlayerPawn) (Variables|Default|GetRopeActiveTag))
      (Tags|RemoveNonReplicatedTagtoPawn (Variables|Default|GetPlayerPawn) (Variables|Default|GetBlockIncomingPickups))
      (EventDispatchers|UnbindEventfromOnPreInvalidateAbilitySystemComponent (Variables|Default|GetPlayerPawn) (EventDispatchers|CreateEvent :!event Custom|OnPlayerPawnPreInvalidateAbilitySystemComponent)))))

(event Custom|OnPlayerPawnPreInvalidateAbilitySystemComponent
  ;[1 0.6595 0.4235] Event Destroyed and OnPrePlayerPawnDetachAbilitySystemComponent - Remove Rope Active and Block Incoming Pickups tags
  (Utilities|IsValid (Variables|Default|GetPlayerPawn)
    (:Is%20Valid
      (Tags|RemoveNonReplicatedTagtoPawn (Variables|Default|GetPlayerPawn) (Variables|Default|GetRopeActiveTag))
      (Tags|RemoveNonReplicatedTagtoPawn (Variables|Default|GetPlayerPawn) (Variables|Default|GetBlockIncomingPickups))
      (EventDispatchers|UnbindEventfromOnPreInvalidateAbilitySystemComponent (Variables|Default|GetPlayerPawn) (EventDispatchers|CreateEvent :!event Custom|OnPlayerPawnPreInvalidateAbilitySystemComponent)))))

(event Custom|End%20Reel
  (bind _rope (Variables|Default|GetRope))
  (bind _rope_2 (Variables|Default|GetRope))
  (bind _start_rope (Variables|Default|GetStart%20Rope))
  (bind _grappler_arrow (Variables|Default|GetGrappler_Arrow))
  (bind _end_rope (Variables|Default|GetEnd%20Rope))
  ;[1 1 1] Reel In
  (Utilities|FlowControl|Sequence
    (:then_0
      (Audio|PlaySoundatLocation (Variables|Default|GetReelSound) (Transformation|GetActorLocation (Variables|Default|GetPlayerPawn)) "0, 0, 0" 2.0 1.5))
    (:then_1
      (Rendering|SetVisibility (Variables|Default|GetGrappler_Arrow) true false))
    (:then_2
      (bind (_direction _reel_lerp_controller _tighten) (Timeline|Reel%20In PlayFromStart)
        (:Update
          (CallFunction|SetStartPointOfRope)
          (Transformation|SetWorldLocation _grappler_arrow (Math|Vector|Lerp(Vector) _end_rope _start_rope _reel_lerp_controller) false false)
          (Transformation|SetWorldRotation _grappler_arrow (Math|Rotator|FindLookatRotation _start_rope _end_rope) false false)
          (bind _distance (Class|MissionBlueprintFunctionLibrary|DistanceBetweenTwoVectors _start_rope (Math|Vector|Lerp(Vector) _end_rope _start_rope _reel_lerp_controller)))
          (Rendering|Material|SetScalarParameterValueonMaterials _rope "MaxLength" _distance)
          (Transformation|SetRelativeScale3D _rope (Math|Vector|MakeVector (/ _distance 1e+02) (Variables|Default|GetRopeThickness) (Variables|Default|GetRopeThickness)))
          (Rendering|Material|SetScalarParameterValueonMaterials _rope_2 "ReturnWiggle" (Math|Float|Lerp 0.0 3.0 _tighten))
          (Rendering|Material|SetScalarParameterValueonMaterials _rope_2 "WiggleAmount" (Math|Float|Lerp 0.0 15.0 _tighten))
          (Transformation|SetWorldRotation (Variables|Default|GetRope) (Math|Rotator|FindLookatRotation (Variables|Default|GetStart%20Rope) (Variables|Default|GetEnd%20Rope)) false false)
          (Utilities|IsValid (Actor|GetOwner)
            (:Is%20Not%20Valid
              (Rendering|SetVisibility (Variables|Default|GetGrappler_Arrow) false false)
              (Actor|DestroyActor))))
        (:Finished
          (Ability|SendGameplayEventtoActor (Variables|Default|GetPlayerPawn) "(TagName=\"Abilities.Grapple.Return\")") ; Plays reattach anim
          (Audio|PlaySoundatLocation (Variables|Default|GetReattachSound) (Transformation|GetActorLocation (Variables|Default|GetPlayerPawn)))
          (Rendering|SetVisibility _grappler_arrow false false)
          (Actor|DestroyActor))))))

(event Custom|Set%20Start%20Point%20of%20Rope
  (bind _returnvalue (Actor|GetOwner))
  (bind (_asb_weap_bow_grappler_athena _bsuccess) (Utilities|Casting|CastToB_Weap_Bow_Grappler_Athena _returnvalue))
  (bind _weaponmesh (Variables|/Script/FortniteGame.FortWeapon|GetWeaponMesh _asb_weap_bow_grappler_athena))
  ;[1 1 1] Set Start Position
  (if _bsuccess
    (Utilities|IsValid _weaponmesh
      (:Is%20Valid
        (bind _output_get (Variables|Default|SetStart%20Rope (Transformation|GetSocketLocation _weaponmesh "Muzzle")))
        (Transformation|SetWorldLocation (Variables|Default|GetRope) _output_get false false)))
    (else
      (Development|PrintString "Owner is NOT Hook Gun - You have Issues"))))

(event EventTick
  (params (DeltaSeconds :type Float%20%28single-precision%29))
  (bind _rope (Variables|Default|GetRope))
  (bind _projectile_actor (Variables|Default|GetProjectile%20Actor))
  (bind _staticmesh3 (Variables|/PrimalGameplay/Items/Exotics/GrapplerBow/B_Prj_Grappler_Secondary.B_Prj_Grappler_Secondary_C|GetStaticMesh3 _projectile_actor))
  ;[1 1 1] Tick Rope Control
  (CallFunction|SetStartPointOfRope)
  (Utilities|IsValid _staticmesh3
    (:Is%20Valid
      (Variables|Default|SetEnd%20Rope (Transformation|GetSocketLocation _staticmesh3 "RopeAttach"))
      (Transformation|SetWorldLocation (Variables|Default|GetRope) (Variables|Default|GetStart%20Rope) false false)
      (bind _distance (Class|MissionBlueprintFunctionLibrary|DistanceBetweenTwoVectors (Variables|Default|GetStart%20Rope) (Variables|Default|GetEnd%20Rope)))
      (Transformation|SetWorldRotation _rope (Math|Rotator|MakeRotfromX (- (Variables|Default|GetEnd%20Rope) (Variables|Default|GetStart%20Rope))) false false)
      (Rendering|Material|SetScalarParameterValueonMaterials _rope "MaxLength" _distance)
      (Transformation|SetRelativeScale3D _rope (Math|Vector|MakeVector (/ _distance 1e+02) (Variables|Default|GetRopeThickness) (Variables|Default|GetRopeThickness))))
    (:Is%20Not%20Valid
      _))
  (if (Variables|Default|GetProjectile%20Gone)
    (Actor|Tick|SetActorTickEnabled false)
    (CallFunction|EndReel)))

(event Custom|TightenRope
  ;[1 1 1] Tighten Rope
  (bind (_direction _newtrack_0) (Timeline|Tight PlayFromStart)
    (:Update
      (Rendering|Material|SetScalarParameterValueonMaterials (Variables|Default|GetRope) "WiggleAmount" (Math|Float|Lerp 0.0 15.0 _newtrack_0)))))

(fn ConstructionScript)
