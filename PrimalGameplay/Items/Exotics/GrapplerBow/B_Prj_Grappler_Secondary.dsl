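; Grappler Bow secondary projectile that pulls the firing player, or their vehicle if allowed, toward the point it hits.
; BeginPlay spawns the rope trail and rope-active effect; on hit it sticks and tightens the rope.
; The server's Travel consumes ammo, sends grapple events and launches via arc-solved velocity plus momentum.
; Skydive launches cancel the grapple, weapon switches destroy the projectile, and a miss yanks back with a cooldown.
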(event Custom|LaunchCheck
  ;[1 1 1] Reset everything if a launchpad is touched while Proj is active
  (bind _jumppadding (CallFunction|HandleLaunch))
  (if _jumppadding
    (Utilities|FlowControl|DoOnce false
      (:Completed
        (Variables|Default|SetLaunchPad true)
        (Variables|Default|SetHookStopped true)
        (Ability|SendGameplayEventtoActor (Variables|Default|GetPlayerPawn) "(TagName=\"Abilities.Grapple.Hit\")" (Utilities|Struct|MakeGameplayEventData :Instigator self :ContextHandle "()" :InstigatorTags "(GameplayTags=)" :TargetTags "(GameplayTags=)" :EventMagnitude 0.0 :TargetData "()"))
        (Development|PrintString "Skah Dahven event sent" true true "(R=1.000000,G=0.270337,B=0.000000,A=1.000000)" 4.0)
        (Actor|DestroyActor (Variables|Default|GetRope))
        (Utilities|FlowControl|Delay 0.1) ; so proj and rope can't be destroyed at the same time
        (Actor|DestroyActor)))))

(event Custom|WeaponCheck
  ;[1 1 1] Destroys projectil if weapon is switched.  This is so we can keep the rope FX looking good, but still let people weapon switch after grappling
  (if (not (Utilities|Equal(Object) (Variables|/Script/FortniteGame.FortPawn|GetCurrentWeapon (Variables|Default|GetPlayerPawn)) (Variables|Default|GetWeaponActor)))
    (Actor|DestroyActor)))

(event EventBeginPlay
  (bind _playerpawn (Variables|Default|GetPlayerPawn))
  ;[1 1 1] Begin Play
  (bind _asfort_player_pawn (Utilities|Casting|CastToFortPlayerPawn (Variables|Actor|GetInstigator)))
  (Variables|Default|SetPlayerPawn _asfort_player_pawn)
  (Tags|AddNonReplicatedTagtoPawn (Variables|Default|GetPlayerPawn) (Variables|Default|GetRopeActiveTag))
  (bind _asb_weap_bow_grappler_athena (Utilities|Casting|CastToB_Weap_Bow_Grappler_Athena (Variables|/Script/FortniteGame.FortPawn|GetCurrentWeapon (Variables|Default|GetPlayerPawn))))
  (Variables|Default|SetWeaponActor _asb_weap_bow_grappler_athena)
  (bind _returnvalue (Game|SpawnActorfromClass :Class "/PrimalGameplay/Items/Exotics/GrapplerBow/B_Grappler_Trail.B_Grappler_Trail_C" :SpawnTransform (Math|Transform|MakeTransform "0, 0, 0") :CollisionHandlingOverride "Undefined" :TransformScaleMethod "OverrideRootScale" :Owner (Variables|Default|GetWeaponActor) :Projectile%20Actor self :PlayerPawn (Variables|Default|GetPlayerPawn)))
  (Variables|Default|SetRope _returnvalue)
  (bind _returnvalue_2 (GameplayEffects|ApplyGameplayEffectToTarget (Ability|GetAbilitySystemComponent (Variables|Default|GetPlayerPawn)) "/Game/Athena/Items/Weapons/Abilities/HookGun/GE_Athena_RopeActive.GE_Athena_RopeActive_C" (Ability|GetAbilitySystemComponent (Variables|Default|GetPlayerPawn)) 1.0 (Utilities|Struct|MakeGameplayEffectContextHandle)))
  (Variables|/PrimalGameplay/Items/Exotics/GrapplerBow/B_Grappler_Trail.B_Grappler_Trail_C|SetRopeGE _returnvalue_2 (Variables|Default|GetRope))
  (Variables|Default|SetFireAImPitch (FortPlayerPawn|GetBaseAimPitchQuantized (Variables|Default|GetPlayerPawn)))
  (Tags|AddNonReplicatedTagtoPawn (Variables|Default|GetPlayerPawn) (Variables|Default|GetBlockIncomingPickups))
  (if (Vehicle|IsinVehicle _playerpawn)
    (bind _returnvalue_3 (Vehicle|GetVehicleActor (Vehicle|GetVehicleInterface _playerpawn) :!class /Script/FortniteGame.FortVehicleInterface))
    (Collision|IgnoreActorwhenMoving (Variables|Default|GetCapsule) _returnvalue_3 true)))

(event Collision|EventHit
  (params (MyComp :type Primitive%20Component%20Object%20Reference) (Other :type Actor%20Object%20Reference) (OtherComp :type Primitive%20Component%20Object%20Reference) (bSelfMoved :type Boolean) (HitLocation :type Vector :default "0, 0, 0") (HitNormal :type Vector :default "0, 0, 0") (NormalImpulse :type Vector :default "0, 0, 0") (Hit :type Hit%20Result%20Structure%20%28by%20ref%29))
  (bind (_bblockinghit _binitialoverlap _time _distance _location _impactpoint _normal _impactnormal _physmat _hitactor) (Collision|BreakHitResult Hit))
  (|CallParentFunction MyComp Other OtherComp bSelfMoved HitLocation HitNormal NormalImpulse Hit)
  ;[1 1 1] Stop event.  Sets variables and calls the travel event
  (if (Utilities|NotEqual(Object) self _hitactor)
    (if (Variables|/Script/FortniteGame.FortPawn|GetbIsDBNO (Variables|Default|GetPlayerPawn))
      (Ability|SendGameplayEventtoActor (Variables|Default|GetPlayerPawn) "(TagName=\"Abilities.Grapple.Return\")") ; Unblock pawn from weapon switching when interp finishes
      (Actor|DestroyActor (Variables|Default|GetRope))
      (else
        (CallFunction|HandleHits Hit)
        (Actor|SetLifeSpan 0.5)
        (if (and (Networking|IsDedicatedServer) (Utilities|IsValid (Game|GetInstigatorController) :!class /Script/Engine.KismetSystemLibrary))
          (CallFunction|Travel (Variables|Default|GetHitLocation))
          (else
            (Development|PrintString "Not the dedicated server")))))))

(event Custom|Travel
  (params (HitLocation :type Vector))
  (bind _playerpawn (Variables|Default|GetPlayerPawn))
  (bind _returnvalue (Vehicle|GetVehicleInterface _playerpawn))
  (bind _returnvalue_2 (Utilities|ToObject(Interface) _returnvalue))
  (bind (_asfort_athena_jackal_vehicle _bsuccess) (Utilities|Casting|CastToFortAthenaJackalVehicle _returnvalue_2))
  (bind _mesh (Variables|/Script/Engine.Character|GetMesh _playerpawn))
  (bind _weaponactor (Variables|Default|GetWeaponActor))
  (bind _playerpawn_2 (Variables|Default|GetPlayerPawn))
  ;[1 1 1] Travel event.  Plays travel montage, sounds and then moves the player.
  (Variables|Default|SetHitLocation HitLocation)
  (if (and (Variables|Default|GetConsumesAmmo) (not (FortWeapon|Ammo|HasInfiniteUses (Variables|Default|GetWeaponActor))))
    (FortWeapon|Ammo|ConsumeAmmo _weaponactor 1 false)
    (else
      (FortWeapon|Ammo|ConsumeAmmo _weaponactor 0 false)))
  (bind _jumppadding (CallFunction|HandleLaunch))
  (if _jumppadding
    ;[1 1 1] Stop event.  Sets variables and calls the travel event
    (CallFunction|LaunchCheck) ; client/server run into disagreements here - this is a failsafe
    (else
      (bind _returnvalue_3 (Zipline|IsZiplining _playerpawn_2))
      (if _returnvalue_3
        (Zipline|EndZiplining _playerpawn_2 false false))
      (Audio|PlaySoundatLocation (select (Pawn|IsLocallyControlled (Variables|Default|GetPlayerPawn)) (Variables|Default|GetStickSound2D) (Variables|Default|GetStickSound)) (Variables|Default|GetHitLocation))
      (CallFunction|JackalCheck) ; This is to fix an animation bug
      (Ability|SendGameplayEventtoActor (Variables|Default|GetPlayerPawn) "(TagName=\"Abilities.Grapple.Hit\")" (Utilities|Struct|MakeGameplayEventData :Instigator self :ContextHandle "()" :InstigatorTags "(GameplayTags=)" :TargetTags "(GameplayTags=)" :EventMagnitude 0.0 :TargetData "()"))
      (Development|PrintString "EventSent")
      (Utilities|IsValid (Fortnite|GetFortPlayerControllerfromActor _playerpawn)
        (:Is%20Valid
          _)
        (:Is%20Not%20Valid
          (if _bsuccess
            (Animation|Montage|PlayMontage _mesh "/Game/Animation/Game/MainPlayer/Combat/Guns/GrapplingHook/GrapplingHook_Travel_Hoverboard_M.GrapplingHook_Travel_Hoverboard_M") ; These two montage nodes are hacks to make grapple anims work correctly for replays
            (else
              (Animation|Montage|PlayMontage _mesh "/Game/Animation/Game/MainPlayer/Combat/Gadgets/Medium/Male/Ashton/Hippo/AshtonHippo_GrappleBow_Travel_M.AshtonHippo_GrappleBow_Travel_M" 1.0 0.0 (CallFunction|PickGrappleMontage (Variables|Default|GetFireAImPitch) _bsuccess))))))
      (if (Vehicle|IsinVehicle (Variables|Default|GetPlayerPawn))
        (if (ScalableFloat|GetValueasBool (Variables|Default|GetPassengerCanGrapple))
          (CallFunction|GrappleVehicle))
        (else
          (Grappling|Grapple)))
      (Utilities|Time|SetTimerbyFunctionName 0 "WeaponCheck" 0.1 true) ; destroy projectile if weapons switched
      ;[1 1 1] Quest Stuff
      (Development|PrintString "GRAPPLED" false false "(R=1.000000,G=0.000000,B=0.877908,A=1.000000)" 1e+01)
      (GameplayEffects|ApplyGameplayEffectToSelf :self (Ability|GetAbilitySystemComponent (Variables|Default|GetPlayerPawn)) :GameplayEffectClass (Variables|Default|GetGE_GrappleQuestCheck) :Level 0.0))))

(event EventDestroyed
  (bind _playerpawn (Variables|Default|GetPlayerPawn))
  (bind _mesh (Variables|/Script/Engine.Character|GetMesh _playerpawn))
  ;[1 1 1] Destroyed
  (bind _playerpawn_2 (Variables|Default|GetPlayerPawn)
    (:then
      (if (Variables|/Script/FortniteGame.FortPawn|GetbIsDBNO _playerpawn_2)
        (Ability|SendGameplayEventtoActor (Variables|Default|GetPlayerPawn) "(TagName=\"Abilities.Grapple.Return\")") ; Unblock pawn from weapon switching when interp finishes
        (bind _rope (Variables|Default|GetRope)
          (:then
            (Actor|DestroyActor _rope))
          (:else))
        (elif (Variables|Default|GetHookStopped)
          (Effects|SpawnEmitteratLocation "/Game/Weapons/FORT_Crossbows/Ashton_Hippo/Effects/P_Ashton_Hippo_Impact_Break.P_Ashton_Hippo_Impact_Break" (Transformation|GetActorLocation) (Math|Conversions|RotationFromXVector (Variables|Default|GetImpactNormal)))
          (Audio|PlaySoundatLocation (Variables|Default|GetRopeDestroySound) (Transformation|GetActorLocation (Variables|Default|GetPlayerPawn)))
          (Actor|DestroyActor (Variables|Default|GetRope))
          (else
            (Animation|Montage|PlayMontage _mesh "/Game/Animation/Game/MainPlayer/Combat/Gadgets/Medium/Male/Ashton/Hippo/AshtonHippo_GrappleBow_ReAttached_CMM_M.AshtonHippo_GrappleBow_ReAttached_CMM_M" 1.2)
            (Audio|SpawnSoundAttached (Variables|Default|GetYankBackSound) _mesh)
            (Variables|/PrimalGameplay/Items/Exotics/GrapplerBow/B_Grappler_Trail.B_Grappler_Trail_C|SetProjectile%20Gone true (Variables|Default|GetRope))
            ;;[1 1 1] Add cooldown GE to the player if they miss.
            (Environment|ApplyGlobalEnvironmentGameplayEffecttoActor _playerpawn (Variables|Default|GetGE_ReturnCooldown))))))
    (:else)))

(event Custom|GrappleVehicle
  (bind _playerpawn (Variables|Default|GetPlayerPawn))
  (bind _returnvalue (Vehicle|GetVehicleInterface _playerpawn))
  (bind _returnvalue_2 (Class|FortVehicleInterface|GetOwnerLocation _returnvalue))
  (bind _momentum (Variables|Default|GetMomentum))
  (bind _vehiclevelocity (Variables|Default|GetVehicleVelocity))
  ;[1 1 1] Vehicle Grappling
  (HandleSticks|HandleMomentum)
  (Variables|Default|SetVehicleVelocity (Vehicle|GetVehicleVelocity _returnvalue))
  (bind _returnvalue_3 (Vehicle|GetVehicleActor _returnvalue :!class /Script/FortniteGame.FortVehicleInterface))
  (Utilities|Casting|CastToFortAthenaJackalVehicle _returnvalue_3
    (:then
      (Variables|VehicleGrappling|SetbInHoverboard true))
    (:CastFailed
      (Variables|VehicleGrappling|SetbInHoverboard false)))
  (bind _outlaunchvelocity (Game|SuggestProjectileVelocityCustomArc _returnvalue_2 (Variables|Default|GetHitLocation) 0.0 (select (Variables|VehicleGrappling|GetbInHoverboard) (Variables|VehicleGrappling|GetJackalArc) (Variables|VehicleGrappling|GetVehicleArc)))) ; Calculates how much force is needed to move actor to target location
  (if (> (Math|Vector|VectorLength (+ (+ _outlaunchvelocity _momentum) _vehiclevelocity)) (Math|Vector|VectorLength (+ _outlaunchvelocity _momentum)))
    (bind _outlaunchvelocity_2 (Game|SuggestProjectileVelocityCustomArc _returnvalue_2 (Variables|Default|GetHitLocation) 0.0 (select (Variables|VehicleGrappling|GetbInHoverboard) (Variables|VehicleGrappling|GetJackalArc) (Variables|VehicleGrappling|GetVehicleArc)))) ; 40% of the distance + 40% of your veloctiy
    (else
      (bind _outlaunchvelocity_3 (Game|SuggestProjectileVelocityCustomArc _returnvalue_2 (Variables|Default|GetHitLocation) 0.0 (select (Variables|VehicleGrappling|GetbInHoverboard) (Variables|VehicleGrappling|GetJackalArc) (Variables|VehicleGrappling|GetVehicleArc)))) ; Go 40% of the distance
    ))
  (Audio|BroadcastAbilityImpactatLocation "/Game/Athena/Sounds/Events/Cube/Cube_LaunchPlayer_Cue.Cube_LaunchPlayer_Cue" _returnvalue_3 (* (select (> (Math|Vector|VectorLength (+ (+ _outlaunchvelocity _momentum) _vehiclevelocity)) (Math|Vector|VectorLength (+ _outlaunchvelocity _momentum))) (+ _outlaunchvelocity_2 (Variables|Default|GetVehicleVelocity)) _outlaunchvelocity_3) (select (Variables|VehicleGrappling|GetbInHoverboard) (Variables|VehicleGrappling|GetJackalForceMultiplier) (Variables|VehicleGrappling|GetVehicleForceMultiplier))) 7.8e+02))

(fn CalculateYankForce
  (bind (_outlocation _outrotation) (Actor|GetActorEyesViewPoint (Variables|Default|GetPlayerPawn)))
  (bind _outlaunchvelocity (Game|SuggestProjectileVelocityCustomArc (Transformation|GetActorLocation (Variables|Default|GetTarget)) (+ _outlocation (* (Math|Vector|GetForwardVector _outrotation) (Variables|Default|GetDistanceFromInstigator))))) ; Finds out how much force is needed to put the target 300 units in front of where instigator is facing
  (if (Utilities|Enum|NotEqual(Enum) (Networking|GetLocalRole (Variables|Default|GetTarget)) "ROLE_SimulatedProxy")
    (Character|LaunchCharacter (Variables|Default|GetTarget) (+ _outlaunchvelocity (Math|Vector|MakeVector (.x (Transformation|GetVelocity (Variables|Default|GetPlayerPawn))) (.y (Transformation|GetVelocity (Variables|Default|GetPlayerPawn))) (Math|Float|Absolute(Float) (.z (Transformation|GetVelocity (Variables|Default|GetPlayerPawn)))))) true true)))

(fn Grapple
  (bind _playerpawn (Variables|Default|GetPlayerPawn))
  (bind _playerpawn_2 (Variables|Default|GetPlayerPawn))
  (bind _charactermovement (Variables|/Script/Engine.Character|GetCharacterMovement _playerpawn_2))
  (bind _playerpawn_3 (Variables|Default|GetPlayerPawn))
  (bind _playerpawn_4 (Variables|Default|GetPlayerPawn))
  (Pawn|Components|CharacterMovement|ClearAccumulatedForces _charactermovement)
  (Components|Movement|StopMovementImmediately _charactermovement)
  (if (not (Variables|Default|GetAirControl))
    (Input|SetIgnoreMoveInput (Game|GetInstigatorController (Variables|Default|GetPlayerPawn)) true))
  (Audio|PlaySoundatLocation (Variables|Default|GetTravelSound) (Transformation|GetActorLocation (Variables|Default|GetPlayerPawn)) "0, 0, 0" 1.0 1.0 0.0 "/Game/Sounds/Attenuation/Fort_Bullet_Impact_Athena_Att.Fort_Bullet_Impact_Athena_Att")
  (HandleSticks|HandleMomentum)
  (FortMovement|CancelJumpAbility _playerpawn false)
  (if (and (AI|Components|NavMovement|IsFalling (Variables|/Script/Engine.Character|GetCharacterMovement _playerpawn)) (Math|Float|InRange(Float) (Variables|Default|GetFireAImPitch) (Variables|Default|GetDownPitchMin) (Variables|Default|GetDownPitchMax)))
    (Development|PrintString "down!")
    (Variables|Default|SetPreLaunchVelocity (Variables|/Script/Engine.MovementComponent|GetVelocity (Pawn|GetMovementComponent (Variables|Default|GetPlayerPawn))))
    (bind _outlaunchvelocity (Game|SuggestProjectileVelocityCustomArc (Transformation|GetActorLocation (Variables|Default|GetPlayerPawn)) (Variables|Default|GetHitLocation) 0.0 0.9))
    (if (Utilities|Enum|NotEqual(Enum) (Networking|GetLocalRole (Variables|Default|GetPlayerPawn)) "ROLE_SimulatedProxy")
      (Character|LaunchCharacter _playerpawn_4 (+ _outlaunchvelocity (Variables|Default|GetPreLaunchVelocity)) true false)
      (Feedback|SpawnGenericCameraLensEffect _playerpawn_4 "/Game/Weapons/FORT_Crossbows/Crossbow/Effects/BP_Hook_Camera_LensEffect.BP_Hook_Camera_LensEffect_C"))
    (else
      (Character|UnCrouch (Variables|Default|GetPlayerPawn))
      (bind _outlaunchvelocity_2 (Game|SuggestProjectileVelocityCustomArc (Transformation|GetActorLocation (Variables|Default|GetPlayerPawn)) (Math|Vector|MakeVector (.x (Variables|Default|GetHitLocation)) (.y (Variables|Default|GetHitLocation)) (+ (.z (Variables|Default|GetHitLocation)) (Variables|Default|GetAdditionalZ))) 0.0 0.425)) ; Calculates how much force is needed to move actor to target location
      (if (Utilities|Enum|NotEqual(Enum) (Networking|GetLocalRole (Variables|Default|GetPlayerPawn)) "ROLE_SimulatedProxy")
        (Character|LaunchCharacter _playerpawn_3 (+ _outlaunchvelocity_2 (select (> (Math|Vector|VectorLength (+ _outlaunchvelocity_2 (Variables|Default|GetMomentum))) (Math|Vector|VectorLength _outlaunchvelocity_2)) (Variables|Default|GetMomentum) "0, 0, 0")) true true) ; Add impulse isn't resetting fall damage when changing momentum 
        (Feedback|SpawnGenericCameraLensEffect _playerpawn_3 "/Game/Weapons/FORT_Crossbows/Crossbow/Effects/BP_Hook_Camera_LensEffect.BP_Hook_Camera_LensEffect_C")
        (Game|Feedback|ClientStartCameraShake :self (Fortnite|GetFortPlayerControllerfromActor _playerpawn_3) :Shake "/Game/Athena/Items/Weapons/Abilities/HookGun/CamerShake_HookGunHitImpact.CamerShake_HookGunHitImpact_C" :Scale 1.0 :PlaySpace "CameraLocal")))))

(fn HandleHits
  (params (Hit :type Hit%20Result%20Structure))
  (bind (_bblockinghit _binitialoverlap _time _distance _location _impactpoint _normal _impactnormal _physmat _hitactor _hitcomponent) (Collision|BreakHitResult Hit))
  (Variables|Default|SetHookStopped true)
  (Variables|Default|SetHitLocation _impactpoint)
  (Class|BGrapplerTrail|TightenRope (Variables|Default|GetRope))
  (Variables|Default|SetImpactNormal _impactnormal)
  (Transformation|SetWorldRotation (Variables|Default|GetAshton_Hippo_Arrow) (Math|Conversions|RotationFromXVector (* _impactnormal -1.0)) false false)
  (Game|Feedback|ClientPlayForceFeedback (Fortnite|GetFortPlayerControllerfromActor (Variables|Default|GetPlayerPawn)) "/Game/Weapons/ForceFeedback/FF_Generic_Throw.FF_Generic_Throw" "None" false false false)
  (bind _asfort_player_pawn (Utilities|Casting|CastToFortPlayerPawn _hitactor)
    (:then
      (Gameplay|ServerAttachtoActor Hit _asfort_player_pawn (Variables|Default|GetOffset%20Distance%20from%20Physics%20Mesh) (Variables|Default|GetOffsetDistanceFromBone) 0.75))
    (:CastFailed
      (HandleSticks|HandleSticks _hitactor _hitcomponent)))
  (Variables|Default|SetStruckActor _hitactor)
  (return 0))

(fn HandleLaunch
  (if (Math|Boolean|ANDBoolean (not (Variables|/Script/FortniteGame.FortPlayerPawn|GetbIsSkydiving (Variables|Default|GetPlayerPawn))) (not (Parachute|GetPendingSkydiveLaunch (Variables|Default|GetPlayerPawn))) (not (Parachute|IsSkydiving (Variables|Default|GetPlayerPawn))))
    (return false)
    (else
      (return true))))

(fn HandleMomentum
  (bind _struckactor (Variables|Default|GetStruckActor))
  (Utilities|IsValid _struckactor
    (:Is%20Valid
      (bind _asfort_player_pawn (Utilities|Casting|CastToFortPlayerPawn _struckactor)
        (:then
          (if (and (Vehicle|IsinVehicle _asfort_player_pawn) false)
            (Variables|Default|SetMomentum (Vehicle|GetVehicleVelocity (Vehicle|GetVehicleInterface _asfort_player_pawn)))
            (else
              _)))
        (:CastFailed
          _))
      (Variables|Default|SetMomentum (Transformation|GetVelocity (Variables|Default|GetStruckActor))))))

(fn HandleSticks
  (params (Object :type Object%20Reference) (Component :type Primitive%20Component%20Object%20Reference))
  (Utilities|FlowControl|Sequence
    (:then_0
      (HandleSticks|HandleStuckBuilding Object Component))
    (:then_1
      (HandleSticks|HandleStuckVehicle Object Component))
    (:then_2
      (HandleSticks|HandleStuckSupplyDrop Object Component))))

(fn HandleStuckBuilding
  (params (Object :type Object%20Reference) (Component :type Primitive%20Component%20Object%20Reference))
  (bind _asbuilding_wall (Utilities|Casting|CastToBuildingWall Object))
  (LocalVariables|SetLocalWall _asbuilding_wall)
  (bind _asstatic_mesh_component (Utilities|Casting|CastToStaticMeshComponent Component))
  (bind _returnvalue (BuildingWall|IsDoorComponent (LocalVariables|GetLocalWall) _asstatic_mesh_component))
  (if _returnvalue
    (Transformation|AttachActorToComponent Component "None" "KeepWorld" "KeepWorld" "KeepWorld" false)
    (Development|PrintString "HitBuildingWall" true true "(R=0.000000,G=0.660000,B=1.000000,A=1.000000)" 15.0)))

(fn HandleStuckProjectiles
  (params (Object :type Object%20Reference) (Component :type Primitive%20Component%20Object%20Reference))
  (Utilities|FlowControl|Sequence
    (:then_0
      (HandleSticks|HandleStuckBuilding Object Component))
    (:then_1
      (HandleSticks|HandleStuckVehicle Object Component))
    (:then_2
      (HandleSticks|HandleStuckSupplyDrop Object Component))))

(fn HandleStuckSupplyDrop
  (params (Object :type Object%20Reference) (Component :type Primitive%20Component%20Object%20Reference))
  (Utilities|FlowControl|Sequence
    (:then_0
      (Utilities|Casting|CastToAthenaSupplyDrop Object)
      (Transformation|AttachActorToComponent Component "None" "KeepWorld" "KeepWorld" "KeepWorld"))
    (:then_1
      (Utilities|Casting|CastToAthenaSupplyDropBalloon Object)
      (Transformation|AttachActorToComponent Component "None" "KeepWorld" "KeepWorld" "KeepWorld" false)))
  (Development|PrintString "Hit Supply Drop" true true "(R=0.000000,G=1.000000,B=0.087717,A=1.000000)" 1e+01))

(fn HandleStuckVehicle
  (params (Object :type Object%20Reference) (Component :type Primitive%20Component%20Object%20Reference))
  (Utilities|Casting|CastToFortAthenaVehicle Object)
  (Transformation|AttachActorToComponent Component "None" "KeepWorld" "KeepWorld" "KeepWorld" false)
  (Development|PrintString "Hit Vehicle" true true "(R=0.000000,G=0.921875,B=0.037376,A=1.000000)" 15.0))

(fn IncapacitateTarget
  (bind _target (Variables|Default|GetTarget))
  (bind _returnvalue (Pawn|GetController _target))
  (Components|Movement|StopMovementImmediately (Variables|/Script/Engine.Character|GetCharacterMovement (Variables|Default|GetTarget)))
  (Input|SetIgnoreMoveInput _returnvalue true)
  ;[1 1 1] Makes target face instigator
  (Pawn|SetControlRotation _returnvalue (Math|Rotator|FindLookatRotation (Transformation|GetActorLocation (Variables|Default|GetTarget)) (Transformation|GetActorLocation (Variables|Default|GetPlayerPawn)))))

(fn JackalCheck
  (bind _playerpawn (Variables|Default|GetPlayerPawn))
  (bind _returnvalue (Vehicle|GetVehicleInterface _playerpawn))
  (bind _returnvalue_2 (Utilities|ToObject(Interface) _returnvalue))
  (Utilities|IsValid _returnvalue_2
    (:Is%20Valid
      (Utilities|Casting|CastToFortAthenaJackalVehicle _returnvalue_2
        (:then
          (GameplayEffects|ApplyGameplayEffectToTarget (Ability|GetAbilitySystemComponent (Variables|Default|GetPlayerPawn)) "/Game/Athena/Items/Weapons/Abilities/HookGun/GE_Athena_HoverboardMontagePlaying.GE_Athena_HoverboardMontagePlaying_C" (Ability|GetAbilitySystemComponent (Variables|Default|GetPlayerPawn)) 1.0 (Utilities|Struct|MakeGameplayEffectContextHandle))
          (return))
        (:CastFailed
          (return))))
    (:Is%20Not%20Valid
      (return))))

(fn OnRep_HookStopped)

(fn OnRep_Momentum)

(fn OnRep_StruckActor)

(fn OnRep_Target)

(fn PickGrappleMontage
  (params (PitchAngle :type Float%20%28double-precision%29) (InHoverboard :type Boolean))
  (bind (_minpitch_6_c27dbdc64650f9ba5ebf09be2564561b _maxpitch_7_75538945434998adec50baa0edf1654a _sectiontoplay_12_2e20dbf74b43f4e8973e488482dca3f7) (Utilities|Struct|BreakToyAngleSelectorInfo (Utilities|Array|Get(aref) (Variables|Default|GetLaunchAngleVariations))))
  (LocalVariables|SetCleanedAngle (Math|Float|ClampAngle 0.0 -9e+01 9e+01))
  (for _array_element (Variables|Default|GetLaunchAngleVariations)
    (if (and (>= (LocalVariables|GetCleanedAngle) (Utilities|Struct|BreakToyAngleSelectorInfo _array_element)) (<= (LocalVariables|GetCleanedAngle) (.MaxPitch_7_75538945434998ADEC50BAA0EDF1654A (Utilities|Struct|BreakToyAngleSelectorInfo _array_element))))
      (return (.SectionToPlay_12_2E20DBF74B43F4E8973E488482DCA3F7 (Utilities|Struct|BreakToyAngleSelectorInfo _array_element)))))
  (return _sectiontoplay_12_2e20dbf74b43f4e8973e488482dca3f7))

(fn PickGrappleMontageSection
  (params (PitchAngle :type Float%20%28double-precision%29))
  (bind (_minpitch_6_c27dbdc64650f9ba5ebf09be2564561b _maxpitch_7_75538945434998adec50baa0edf1654a _sectiontoplay_12_2e20dbf74b43f4e8973e488482dca3f7) (Utilities|Struct|BreakToyAngleSelectorInfo (Utilities|Array|Get(aref) (Variables|Default|GetLaunchAngleVariations))))
  (LocalVariables|SetCleanedAngle (Math|Float|ClampAngle PitchAngle -9e+01 9e+01))
  (for _array_element (Variables|Default|GetLaunchAngleVariations)
    (if (and (>= (LocalVariables|GetCleanedAngle) (Utilities|Struct|BreakToyAngleSelectorInfo _array_element)) (<= (LocalVariables|GetCleanedAngle) (.MaxPitch_7_75538945434998ADEC50BAA0EDF1654A (Utilities|Struct|BreakToyAngleSelectorInfo _array_element))))
      (return (.SectionToPlay_12_2E20DBF74B43F4E8973E488482DCA3F7 (Utilities|Struct|BreakToyAngleSelectorInfo _array_element)))))
  (return _sectiontoplay_12_2e20dbf74b43f4e8973e488482dca3f7))

(fn PlayStickSounds
  (params (Location :type Vector) (HitActor :type Actor%20Object%20Reference))
  (Utilities|IsValid (Variables|Default|GetPlayerPawn)
    (:Is%20Valid
      (Utilities|IsValid HitActor
        (:Is%20Valid
          ;[1 1 1] Plays if an actor is hit
          (Utilities|FlowControl|Sequence
            (:then_0
              (if (Pawn|IsLocallyControlled (Variables|Default|GetPlayerPawn))
                (Audio|PlaySound2D (Variables|Default|GetStuckToPlayerSound)) ; Plays on instigator
              ))
            (:then_1
              (if (Pawn|IsLocallyControlled (Utilities|Casting|CastToFortPlayerPawn HitActor))
                (Audio|PlaySound2D (Variables|Default|GetStuckToPlayerSound)) ; Plays on target
                (else
                  _)))))
        (:Is%20Not%20Valid
          (if (Pawn|IsLocallyControlled (Variables|Default|GetPlayerPawn))
            ;[1 1 1] Plays if env is hit
            (Audio|PlaySound2D (Variables|Default|GetStickSound))
            (else
              _)))))
    (:Is%20Not%20Valid
      _))
  (Audio|PlaySoundatLocation (Variables|Default|GetStickSound) Location))

(fn ConstructionScript
  (Variables|/Script/Engine.ProjectileMovementComponent|SetInitialSpeed (ScalableFloat|GetValueatLevel (Variables|Default|GetRow_Speed)) (Variables|Movement|GetProjectileMovementComponent))
  (|CallParentFunction))
