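; Grappler bow weapon Blueprint that stops its owner from using the alt-fire grapple while charging a shot.
; Charge start applies a blocking gameplay effect to the instigator.
; Charge end, local unequip and projectile spawn apply a second effect that removes the block, the last as a safety net.
; Trajectory setup also turns on valid-target checking on the parent bow's trajectory preview.
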(event EventOnChargeUp
  (bind _returnvalue (Game|GetInstigator))
  ;[1 1 1] Apply effect to owner while they are charging the weapon up so that they can not grapple
  (Utilities|IsValid _returnvalue
    (:Is%20Valid
      (Environment|ApplyGlobalEnvironmentGameplayEffecttoActor _returnvalue (Variables|Default|GetGE_AltFireBlockCharge)))
    (:Is%20Not%20Valid
      _))
  (|CallParentFunction))

(event FortWeapon|EventOnLocalUnEquip
  (bind _returnvalue (Game|GetInstigator))
  ;[1 1 1] Remove the effect blocking the altfire ability when the charge stops or on unequip by applying a GE that removes the blocking one
  (Utilities|IsValid _returnvalue
    (:Is%20Valid
      (Environment|ApplyGlobalEnvironmentGameplayEffecttoActor _returnvalue (Variables|Default|GetGE_AltFireBlockChargeRemove)))
    (:Is%20Not%20Valid
      _))
  (|CallParentFunction))

(event EventOnEndCharge
  (bind _returnvalue (Game|GetInstigator))
  ;[1 1 1] Remove the effect blocking the altfire ability when the charge stops or on unequip by applying a GE that removes the blocking one
  (Utilities|IsValid _returnvalue
    (:Is%20Valid
      (Environment|ApplyGlobalEnvironmentGameplayEffecttoActor _returnvalue (Variables|Default|GetGE_AltFireBlockChargeRemove)))
    (:Is%20Not%20Valid
      _))
  (|CallParentFunction))

(event FortWeapon|EventOnSpawnProjectile
  (params (SpawnProjectile :type Fort%20Projectile%20Base%20Object%20Reference))
  (bind _returnvalue (Game|GetInstigator))
  ;[1 1 1] Unblock After Firing the projectile just in case.
  (Utilities|IsValid _returnvalue
    (:Is%20Valid
      (Environment|ApplyGlobalEnvironmentGameplayEffecttoActor _returnvalue (Variables|Default|GetGE_AltFireBlockChargeRemove)))
    (:Is%20Not%20Valid
      _))
  (|CallParentFunction SpawnProjectile))

(event EventSetupTrajectoryProjectile
  (|CallParentFunction)
  (Variables|/Game/Athena/Items/Weapons/Abilities/ExplosiveBow/BP_ProjectileTrajectory_Bow_Generic.BP_ProjectileTrajectory_Bow_Generic_C|SetCheckForValidTarget true (Variables|ParentBow|GetBowTrajectory)))

(fn ConstructionScript
  (|CallParentFunction))
